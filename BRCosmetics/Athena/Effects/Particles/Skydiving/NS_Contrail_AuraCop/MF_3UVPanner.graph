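# MF_3UVPanner is a material function in the AuraCop skydiving contrail folder that outputs three panned UV sets.
# Rotation sets the base angle, and the layers are spaced a third of a turn apart, with matching offsets on the speed-scaled time.
# A static switch either rotates UVs about their center and pans along U, or leaves UVs intact and rotates the pan direction.
# Another switch selects fractional-part panning.
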MaterialFunction /BRCosmetics/Athena/Effects/Particles/Skydiving/NS_Contrail_AuraCop/MF_3UVPanner
graph {nodes [
  n1 = FunctionInput(Preview=TextureCoordinate()) {bUsePreviewValueAsDefault=true, inputName=UV, inputType=FunctionInput_Vector2}
  n2 = MaterialFunctionCall(Float2=Subtract(A=@n1) {constB=0.5}) {materialFunction=&BreakOutFloat2Components}
  n3 = FunctionInput() {
   bUsePreviewValueAsDefault=true
   description=Adds an initial rotation to the UVs
   inputName=Rotation
   inputType=FunctionInput_Scalar
   sortPriority=1
  }
  n4 = Add(A=@n3) {constB=0.333}
  n5 = Sine(None=@n4)
  n6 = Multiply(A=@n2.R, B=@n5)
  n7 = Cosine(None=@n4)
  n8 = Multiply(A=@n2.G, B=@n7)
  Add(A=@n6, B=@n8)
  n9 = Add(A=@n3) {constB=0.666}
  n10 = Sine(None=@n9)
  n11 = Multiply(A=@n2.R, B=@n10)
  n12 = Cosine(None=@n9)
  n13 = Multiply(A=@n2.G, B=@n12)
  Add(A=@n11, B=@n13)
  n14 = Sine(None=@n3)
  n15 = Multiply(A=@n2.R, B=@n14)
  n16 = Cosine(None=@n3)
  n17 = Multiply(A=@n2.G, B=@n16)
  Add(A=@n15, B=@n17)
  n18 = Multiply(A=@n2.R, B=@n16)
  n19 = Multiply(A=@n2.G, B=@n14)
  n20 = Add(A=@n18, B=@n19)
  n21 = Subtract(A=@n17, B=@n15)
  n22 = FunctionInput(Preview=StaticBool()) {
   bUsePreviewValueAsDefault=true
   description="Switches between rotating UVs (True) or rotating Speed (False), defaults to False"
   inputName=Rotate UVs
   inputType=FunctionInput_StaticBool
   sortPriority=4
  }
  n23 = StaticSwitch(True=Add(A=AppendVector(A=@n20, B=@n21)) {constB=0.5}, False=@n1, Value=@n22)
  n24 = Constant2Vector() {r=1}
  n25 = StaticSwitch(True=@n24, False=AppendVector(A=Cosine(None=@n3), B=Sine(None=@n3)), Value=@n22)
  n26 = FunctionInput() {
   bUsePreviewValueAsDefault=true
   description=Panner speed
   inputName=Speed
   inputType=FunctionInput_Scalar
   sortPriority=2
   previewValue=<1 1 0 1>
  }
  n27 = FunctionInput(Preview=Time()) {
   bUsePreviewValueAsDefault=true
   description="Custom time input to allow for offsets, defaults to Game Time"
   inputName=Time
   inputType=FunctionInput_Scalar
   sortPriority=3
  }
  n28 = Multiply(A=@n26, B=@n27)
  n29 = FunctionInput(Preview=StaticBool()) {bUsePreviewValueAsDefault=true, description=Defaults to False, inputName=Fractional Part, inputType=FunctionInput_StaticBool, sortPriority=4}
  FunctionOutput(None=StaticSwitch(True=Panner(Coordinate=@n23, Time=@n28, Speed=@n25) {bFractionalPart=true}, False=Panner(Coordinate=@n23, Time=@n28, Speed=@n25), Value=@n29)) {outputName=UV1}
  n30 = Multiply(A=@n2.R, B=@n7)
  n31 = Multiply(A=@n2.G, B=@n5)
  n32 = Add(A=@n30, B=@n31)
  n33 = Subtract(A=@n8, B=@n6)
  n34 = StaticSwitch(True=Add(A=AppendVector(A=@n32, B=@n33)) {constB=0.5}, False=@n1, Value=@n22)
  n35 = StaticSwitch(True=@n24, False=AppendVector(A=Cosine(None=@n4), B=Sine(None=@n4)), Value=@n22)
  n36 = Add(A=@n28) {constB=0.333}
  FunctionOutput(None=StaticSwitch(True=Panner(Coordinate=@n34, Time=@n36, Speed=@n35) {bFractionalPart=true}, False=Panner(Coordinate=@n34, Time=@n36, Speed=@n35), Value=@n29)) {outputName=UV2, sortPriority=1}
  n37 = Multiply(A=@n2.R, B=@n12)
  n38 = Multiply(A=@n2.G, B=@n10)
  n39 = Add(A=@n37, B=@n38)
  n40 = Subtract(A=@n13, B=@n11)
  n41 = StaticSwitch(True=Add(A=AppendVector(A=@n39, B=@n40)) {constB=0.5}, False=@n1, Value=@n22)
  n42 = StaticSwitch(True=@n24, False=AppendVector(A=Cosine(None=@n9), B=Sine(None=@n9)), Value=@n22)
  n43 = Add(A=@n28) {constB=0.666}
  FunctionOutput(None=StaticSwitch(True=Panner(Coordinate=@n41, Time=@n43, Speed=@n42) {bFractionalPart=true}, False=Panner(Coordinate=@n41, Time=@n43, Speed=@n42), Value=@n29)) {outputName=UV3, sortPriority=2}
  Subtract(A=@n19, B=@n18)
  Subtract(A=@n38, B=@n37)
  Subtract(A=@n31, B=@n30)
 ]}
