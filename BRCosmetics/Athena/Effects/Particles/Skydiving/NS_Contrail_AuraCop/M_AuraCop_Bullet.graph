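# Default Lit material for Niagara mesh particles, apparently the bullets in the AuraCop skydiving contrail effect.
# Rounded texture coordinates blend TopColor and BaseColor as a hard split rather than a gradient.
# An inverted Fresnel compares the view-space vertex normal with a fixed view-axis vector instead of the true camera vector.
# That mask tints GlowColour and is scaled by the Glow dynamic parameter, allowing per-particle glow control.
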Material /BRCosmetics/Athena/Effects/Particles/Skydiving/NS_Contrail_AuraCop/M_AuraCop_Bullet
graph {
 properties {bUsedWithNiagaraMeshParticles=true, pixelDepthOffsetMode=PDOM_Legacy, usedShadingModels=Default Lit}
 nodes [
  Constant3Vector() {constant=<rgba 0 0 1 0>}
  Constant() {r=1}
  Constant() {r=0.7}
  Constant() {r=0.3}
  LinearInterpolate(
   A=VectorParameter() {parameterName=TopColor, defaultValue=<rgba 0.224 0.198 0.1703 1>}.RGB
   B=VectorParameter() {parameterName=BaseColor, defaultValue=<rgba 1 0.575 0.15 0>}.RGB
   Alpha=Round(None=TextureCoordinate())
  )
  n1 = Constant() {r=4}
  n2 = Constant3Vector() {constant=<rgba 0 0 -1 0>}
  n3 = Normalize(VectorInput=Transform(None=VertexNormalWS()) {transformSourceType=TRANSFORMSOURCE_World, transformType=TRANSFORM_View})
  Multiply(
   A=DynamicParameter() {defaultValue=<rgba 0 0 1 1>, paramNames [Glow,Param2,Param3,Param4]}.Glow
   B=Multiply(
    A=MaterialFunctionCall(Exponent=@n1, Camera Vector=@n2, Vertex / Pixel Normal=@n3) {materialFunction=&MF_Fresnel_FX}.Fresnel Inverted
    B=VectorParameter() {parameterName=GlowColour, defaultValue=<rgba 10 5.0014 0 1>}.RGB
   )
  )
  VertexColor()
 ]
}
