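# Niagara skydiving contrail "AuraCop": an energy sphere around the player is struck by looping bullets that deflect away.
# NE_Bullets uses a modulo progress alpha to drive scripted two-phase paths, flying in, then tumbling outward, with impact sprites.
# NE_BulletRibbons follows those bullets through particle attribute reads, drawing two-point trails.
# Dive alpha intensifies the sphere, ParaGlideLeanAlpha fades in speedlines and alters deflection, and front-end modifiers adjust placement and scale.
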NiagaraSystem /BRCosmetics/Athena/Effects/Particles/Skydiving/NS_Contrail_AuraCop/NS_Contrail_AuraCop
dependencies {
 usedDataInterfaces [
  &NiagaraDataInterfaceParticleRead
  &NiagaraDataInterfaceCurve
  &NiagaraDataInterfaceColorCurve
  &NiagaraDataInterfaceVector2DCurve
  &NiagaraDataInterfaceCamera
  &NiagaraDataInterfaceSkeletalMesh
 ]
 usedDynamicInputs [
  /Niagara/DynamicInputs/Lerp/SmoothLerpOverTime_Float
  &TransformPosition
  &MakeFloatFromVector
  &Abs_Float
 ]
 usedModules [
  &self:SystemSpawnScript.SystemScriptSource.SystemScriptGraph.NiagaraNodeAssignment_9.SetVariables_#1
  &self:SystemSpawnScript.SystemScriptSource.SystemScriptGraph.NiagaraNodeAssignment_0.SetVariables_#2
  &self:SystemSpawnScript.SystemScriptSource.SystemScriptGraph.NiagaraNodeAssignment_1.SetVariables_#3
  &self:SystemSpawnScript.SystemScriptSource.SystemScriptGraph.NiagaraNodeAssignment_2.SetVariables_#4
  &self:SystemSpawnScript.SystemScriptSource.SystemScriptGraph.NiagaraNodeAssignment_5.SetVariables_#5
  &SystemState
  &self:SystemSpawnScript.SystemScriptSource.SystemScriptGraph.NiagaraNodeAssignment_6.SetVariables_#6
  &self:SystemSpawnScript.SystemScriptSource.SystemScriptGraph.NiagaraNodeAssignment_4.SetVariables_#7
  &self:SystemSpawnScript.SystemScriptSource.SystemScriptGraph.NiagaraNodeAssignment_7.SetVariables_#8
  &self:SystemSpawnScript.SystemScriptSource.SystemScriptGraph.NiagaraNodeAssignment_3.SetVariables_#9
  &self:SystemSpawnScript.SystemScriptSource.SystemScriptGraph.NiagaraNodeAssignment_8.SetVariables_#10
 ]
}
emitters {
 Energy_Sphere {
  emitterUpdateScript [
   &EmitterState(
    Inactive Response=::NewEnumerator1
    Life Cycle Mode=::NewEnumerator1
    Loop Behavior=::NewEnumerator1
    Loop Duration Mode=::NewEnumerator1
    UseLoopDelay=false
   )
   &SpawnBurst_Instantaneous(Spawn Count=1, Use Spawn Probability=true)
  ]
  particleSpawnScript [
   &InitializeParticle(
    Color=<rgba 2 1.3131 0.228 1>
    Lifetime=0.3
    Mesh Scale Mode=::NewEnumerator3
    Mesh Uniform Scale=&Multiply_Float(A=$System.SphereRadius, B=&FN_FrontEndFloatModifier(Gameplay=0.019, Front End=0.021))
    Sprite Size Mode=::NewEnumerator3
    Uniform Sprite Size=&Multiply_Float(A=$System.SphereRadius, B=2.9)
    UsePositionOffset=true
    Write Material Random=true
   ) {module=InitializeParticle001}
   &CameraOffset(Camera Offset Amount=20)
   /SetVariables(Particles.FlareOffset=-50) {isSetParameters=true}
   &InitialMeshOrientation(Mesh Orientation Mode=::NewEnumerator7, Use Rotation Vector=true)
  ]
  particleUpdateScript [
   &ParticleState(Kill Particles When Lifetime Has Elapsed=false)
   &SampleSkeletalMesh(
    AdditionalYawPitchRoll=false
    Bone Sampling Mode:ENiagaraBoneSamplingMode=::NewEnumerator4
    Sampled Position Coordinate Space:ENiagaraCoordinateSpace=::Local
    Skeletal Mesh:NiagaraDataInterfaceSkeletalMesh=$System.Skeletal Mesh
   )
   &ScaleColor(
    Scale Alpha=&FloatFromCurve(
     FloatCurve=DataInterface() {propertyValues {Curve {keys [{arriveTangent=0.1268, interpMode=RCIM_Cubic, leaveTangent=0.1268, tangentMode=RCTM_User, time=0.3, value=0.5}, <at 1 1 RCIM_Cubic>]}}}
     Scale Curve=0.25
    )
    Scale RGB=&MakeVectorFromLinearColorRGB(Color=&MultiplyLinearColor(
      Linear Color=<rgba 1 1 1 1>
      Scale Factor=&MakeVector4_FromFloat(Float=&Lerp_Float(
        A=1
        B=3
        Alpha=&FloatFromCurve(
         FloatCurve=DataInterface() {propertyValues {Curve {keys [
             {arriveTangent=2.5459, interpMode=RCIM_Cubic, leaveTangent=2.5459, tangentMode=RCTM_User, time=0.6, value=0}
             {interpMode=RCIM_Cubic, tangentMode=RCTM_User, time=0.8, value=1}
             <at 1.1 0 RCIM_Cubic>
            ]}}}
         CurveIndex=$Engine.System.Age
        )
       ))
     ))
    ScaleA=true
    ScaleRGB=true
   )
   /SetVariables(Particles.Position=&Add_Vector(
     A=&FN_FrontEndVector3Modifier(Gameplay=$System.SphereOffset, Front End=&Lerp_Vector(A=<0 0 0>, B=<0 0 0>, Alpha=$System.SmoothDiveAlpha))
     B=&FN_FrontEndVector3Modifier(Gameplay=<0 0 0>, Front End=<0 0 90>)
    )) {isSetParameters=true}
   /SetVariables(Particles.Color002=&MakeLinearColorFromVectorAndFloat(
     "Vector (RGB)"=&MakeVectorFromLinearColorRGB(Color=<rgba 1 1 1 1>)
     "Float (Alpha)"=&FloatFromCurve(FloatCurve=DataInterface() {propertyValues {Curve {keys [<at 0.5 0>,<at 1 1>]}}}, CurveIndex=$Engine.System.Age)
    )) {isSetParameters=true}
   &ScaleMeshSize(Scale Factor=&VectorFromFloat(Value=&FloatFromCurve(FloatCurve=DataInterface() {propertyValues {Curve {keys [
          {arriveTangent=2.3018, interpMode=RCIM_Cubic, leaveTangent=2.3018, tangentMode=RCTM_User, time=0, value=0.2}
          {arriveTangent=0.1704, interpMode=RCIM_Cubic, leaveTangent=0.1704, tangentMode=RCTM_User, time=1, value=1}
         ]}}})))
   &DynamicMaterialParameters(
    Index 0 Param 1=&FloatFromCurve(
     FloatCurve=DataInterface() {propertyValues {Curve {keys [
         {arriveTangent=-0.1503, interpMode=RCIM_Cubic, leaveTangent=-0.1503, tangentMode=RCTM_User, time=0, value=1}
         {arriveTangent=-1.1654, interpMode=RCIM_Cubic, leaveTangent=-1.1654, tangentMode=RCTM_User, time=0.8, value=0}
        ]}}}
     CurveIndex=$Engine.System.Age
    )
    Index 0 Param 2=1
    Index 2 Param 1=3
    Index 2 Param 2=0.5
    Index 2 Param 3=0.3
    Index 2 Param 4=0.5
    Index 3 Param 2=&FloatFromCurve(FloatCurve=DataInterface() {propertyValues {Curve {keys [<at 0 1>,<at 1 0>]}}})
    Module.Index0Param0Write=true
    Module.Index0Param1Write=true
    Module.Index0Param2Write=true
    Module.Index0Param3Write=true
    Module.Index2Param0Write=true
    Module.Index2Param1Write=true
    Module.Index2Param2Write=true
    Module.Index2Param3Write=true
    Module.Index3Param0Write=true
    Module.Index3Param1Write=true
    Module.Index3Param2Write=true
    Module.Index3Param3Write=true
    Write Parameter Index 2=true
    Write Parameter Index 3=true
   )
   &FN_Mesh_Attributes(
    Color=&Multiply_LinearColorByFloat(
     Linear Color=&LerpLinearColors(StartColor=<rgba 3 2.1806 1.2764 1>, EndColor=<rgba 8 5.126 2.4132 1>, LerpFactor=$System.SmoothDiveAlpha)
     Float=$System.FadeIn
    )
    Create Color Attribute=true
    Create Orientation Attribute=true
    Create Scale Attribute=true
    Orientation:Quat4f=/BRCosmetics/Athena/Effects/Particles/Skydiving/NS_Contrail_Apprentice/OrientMestToVectorFromCustomRot(
     InitialRotation:Quat4f=$Particles.Initial.MeshOrientation
     Look At Direction:Vector3f=<0 0 1>
     Reference Axis Vector:Vector3f=<-1 0 0>
     RotationVector:Vector3f=<0 0 0>
    )
    Scale=&Multiply_VectorByFloat(Vector=$Particles.Scale, Float=0.5)
    Write to Module Namespace=true
   )
   &FN_Mesh_Attributes(
    Color=&MakeLinearColorFromVectorAndFloat(
     "Vector (RGB)"=&MakeVectorFromLinearColorRGB(Color=<rgba 1 0.4094 0 1>)
     "Float (Alpha)"=&Multiply_Float(A=&Lerp_Float(A=0.01, B=0.02, Alpha=$System.SmoothDiveAlpha), B=1)
    )
    Create Color Attribute=true
    Create Orientation Attribute=true
    Create Position Attribute=true
    Create Scale Attribute=true
    Orientation:Quat4f=&OrientMestToVectorFromCustomRot(
     InitialRotation:Quat4f=$Particles.Initial.MeshOrientation
     Look At Direction:Vector3f=<0 0 1>
     Reference Axis Vector:Vector3f=<0 0 1>
     RotationVector:Vector3f=<0 0 0>
    )
    Position:Vector3f=&AddVectorToPosition(Position=Linked() {linkedVariable {name=System.SphereOffset, type=Vector3f}}, Vector=<0 0 20>)
    Scale=&Multiply_VectorByFloat(Vector=$Particles.Scale, Float=1.2)
    Write to Module Namespace=true
   ) {module=FN_Mesh_Attributes001}
   &FN_Mesh_Attributes(
    Color=&MakeLinearColorFromVectorAndFloat(
     "Vector (RGB)"=&MakeVectorFromLinearColorRGB(Color=&MultiplyLinearColor(
       Linear Color=<rgba 1 0.3255 0 1>
       Scale Factor=&MakeVector4_FromFloat(Float=&Lerp_Float(
         A=0.7
         B=2
         Alpha=&FloatFromCurve(
          FloatCurve=DataInterface() {propertyValues {Curve {keys [
              {arriveTangent=2.5459, interpMode=RCIM_Cubic, leaveTangent=2.5459, tangentMode=RCTM_User, time=0, value=0}
              {interpMode=RCIM_Cubic, tangentMode=RCTM_User, time=0.2, value=1}
              <at 1.1 0 RCIM_Cubic>
             ]}}}
          CurveIndex=$Engine.System.Age
         )
        ))
      ))
     "Float (Alpha)"=0.6
    )
    Create Color Attribute=true
    Create Position Attribute=true
    Create Scale Attribute=true
    Scale=&Multiply_VectorByFloat(
     Vector=$Particles.Scale
     Float=&FloatFromCurve(
      FloatCurve=DataInterface() {propertyValues {Curve {keys [
          {arriveTangent=3.2262, interpMode=RCIM_Cubic, leaveTangent=3.2262, tangentMode=RCTM_User, time=0, value=0}
          {interpMode=RCIM_Cubic, tangentMode=RCTM_User, time=0.5, value=1}
         ]}}}
      CurveIndex=$System.Age
      Scale Curve=3
     )
    )
    Write to Module Namespace=true
   ) {module=FN_Mesh_Attributes002}
   /SetVariables(
    Particles.HaloColor:LinearColor=&MultiplyLinearColor(
     Linear Color=$Particles.Color
     Scale Factor=&MakeFloat4FromVectorAndFloat(
      "Vector (RGB)"=<1 1 1>
      "Float (Alpha)"=&FloatFromCurve(FloatCurve=DataInterface() {propertyValues {Curve {keys [<at 0.5 0>,<at 1 1>]}}}, CurveIndex=$Engine.System.Age)
     )
    )
    Particles.HaloDynMat:Vector4f=<0.6 0 1 0>
   ) {isSetParameters=true}
  ]
  rendererClasses [&NiagaraMeshRendererProperties, &NiagaraSpriteRendererProperties]
  renderers [
   {rendererClass=&NiagaraMeshRendererProperties, rendererIndex=0}
   {rendererClass=&NiagaraMeshRendererProperties, rendererIndex=1}
   {rendererClass=&NiagaraMeshRendererProperties, rendererIndex=2}
   {rendererClass=&NiagaraMeshRendererProperties, rendererIndex=3}
   {rendererClass=&NiagaraMeshRendererProperties, rendererIndex=4}
   {rendererClass=&NiagaraSpriteRendererProperties, rendererIndex=5}
  ]
 }
 NE_BulletRibbons {
  emitterSpawnScript [/SetVariables(Emitter.ParticlePerRibbon=2) {isSetParameters=true}]
  emitterUpdateScript [
   &EmitterState(Life Cycle Mode=::NewEnumerator1, Loop Behavior=::NewEnumerator1, Loop Duration Mode=::NewEnumerator1, UseLoopDelay=false)
   &SpawnBurst_Instantaneous(
    Spawn Count=&Multiply_Int(A=$System.BulletNumber, B=$Emitter.ParticlePerRibbon)
    Use Spawn Probability=false
    UseLoopCountLimit=false
   )
  ]
  particleSpawnScript [
   &PartitionParticles(
    Normalized Index In Partition Includes One=true
    Number Of Partitions=$System.BulletNumber
    Particles Per Partition:NiagaraInt32=$Emitter.ParticlePerRibbon
    Partition Mode:ENiagara_PartitionMode=::NewEnumerator1
    Write Partition Index to Ribbon Link Order:ENiagara_AttributeSamplingApplyOutput=::NewEnumerator0
    Write Partition to Ribbon ID=::NewEnumerator0
   )
   &InitializeParticle(
    Color=<rgba 1 0.6969 0.194 0.15>
    Lifetime=5
    Ribbon Width=30
    Ribbon Width Mode=::NewEnumerator1
    UsePositionOffset=false
    Write Material Random=true
   )
   &self:GetBulletParams(BulletReader=DataInterface() {propertyValues {EmitterBinding {bindingMode=Other, emitterName=NE_Bullets}}}) {module=GetBulletParams001}
   /SetVariables(Particles.Position=$Output.GetBulletParams001.BulletPos) {isSetParameters=true}
   /SetVariables(Particles.RibbonU0Override=$Particles.PartitionParticles.Normalized Index In Partition) {isSetParameters=true}
  ]
  particleUpdateScript [
   &ParticleState(Kill Particles When Lifetime Has Elapsed=false)
   &self:GetBulletParams(BulletReader=DataInterface() {propertyValues {EmitterBinding {bindingMode=Other, emitterName=NE_Bullets}}}) {module=GetBulletParams}
   /SetVariables(Particles.Position=/Niagara/DynamicInputs/TypeConversions/MakeCustomPositionFromBool(
     Bool:NiagaraBool=&SetBoolByFloatComparison(A=$Output.GetBulletParams.BulletProgressAlpha, B=0.05, Comparison Type=::NewEnumerator6)
     True Position:NiagaraPosition=$Output.GetBulletParams.BulletPos
     False Position:NiagaraPosition=&Lerp_Vector(
      A=$Output.GetBulletParams.BulletPos
      B=&Lerp_Vector(
       A=Linked() {linkedVariable {name=Particles.Position, type=NiagaraPosition}}
       B=$Output.GetBulletParams.BulletPos
       Alpha=&Multiply_Float(
        A=$Engine.DeltaTime
        B=&Lerp_Float(
         A=6
         B=3
         Alpha=&FloatFromCurve(
          FloatCurve=DataInterface() {propertyValues {Curve {keys [
              {arriveTangent=1.9304, interpMode=RCIM_Cubic, leaveTangent=1.9304, tangentMode=RCTM_User, time=0, value=0}
              {interpMode=RCIM_Cubic, tangentMode=RCTM_User, time=1, value=1}
             ]}}}
          CurveIndex=$Particles.PartitionParticles.Normalized Index In Partition
         )
        )
       )
      )
      Alpha=&FloatFromCurve(
       FloatCurve=DataInterface() {propertyValues {Curve {keys [<at 0 0>,<at 0.5 1>]}}}
       CurveIndex=$Particles.PartitionParticles.Normalized Index In Partition
      )
     )
    )) {isSetParameters=true}
   &ScaleRibbonWidth(Ribbon Width Scale=&Multiply_Float(
     A=&FloatFromCurve(
      FloatCurve=DataInterface() {propertyValues {Curve {keys [<at 0.2 0>,<at 0.3 1>,<at 0.5 1>,<at 0.8 0>]}}}
      CurveIndex=$Output.GetBulletParams.BulletProgressAlpha
     )
     B=&FloatFromCurve(
      FloatCurve=DataInterface() {propertyValues {Curve {keys [<at 0 1>,<at 0.5 1>,<at 1 1>]}}}
      CurveIndex=$Particles.Initial.PartitionParticles.Normalized Index In Partition
     )
    ))
   &ScaleColor(
    Scale Alpha=&FloatFromCurve(
     FloatCurve=DataInterface() {propertyValues {Curve {keys [<at 0.2 0>,<at 0.3 1>,<at 0.45 1>,<at 0.5 0>]}}}
     CurveIndex=$Output.GetBulletParams.BulletProgressAlpha
    )
    ScaleA=true
    ScaleRGB=true
   )
   /SetVariables(Particles.SlowMoTrailColor:LinearColor=&MultiplyLinearColor(Linear Color=$Particles.Color, Scale Factor=&MakeVector4_FromLinearColor(LinearColor=<rgba 1 1 1 0.5>))) {isSetParameters=true}
   /SetVariables(Particles.SlowMoTrailWidth:NiagaraFloat=&Multiply_Float(A=$Particles.RibbonWidth, B=0.5)) {isSetParameters=true}
   &FN_Sprite_Attributes(
    Color=&MakeLinearColorFromVectorAndFloat(
     "Vector (RGB)"=&MakeVectorFromLinearColorRGB(Color=<rgba 2 0.7496 0 1>)
     "Float (Alpha)"=&FloatFromCurve(FloatCurve=DataInterface() {propertyValues {Curve {keys [<at 0.5 1>,<at 0.75 0>]}}}, CurveIndex=$Output.GetBulletParams.BulletProgressAlpha)
    )
    Create Color Attribute=true
    Create Position Attribute=true
    Position:NiagaraPosition=&Lerp_Vector(
     A=$Output.GetBulletParams.BulletPos
     B=&Lerp_Vector(
      A=$Output.GetBulletParams.BulletImpactPos
      B=$Output.GetBulletParams.BulletPos
      Alpha=&Multiply_Float(
       A=$Engine.DeltaTime
       B=&Lerp_Float(
        A=6
        B=3
        Alpha=&FloatFromCurve(
         FloatCurve=DataInterface() {propertyValues {Curve {keys [
             {arriveTangent=1.9304, interpMode=RCIM_Cubic, leaveTangent=1.9304, tangentMode=RCTM_User, time=0, value=0}
             {interpMode=RCIM_Cubic, tangentMode=RCTM_User, time=1, value=1}
            ]}}}
         CurveIndex=$Particles.PartitionParticles.Normalized Index In Partition
        )
       )
      )
     )
     Alpha=&FloatFromCurve(
      FloatCurve=DataInterface() {propertyValues {Curve {keys [<at 0 0>,<at 0.5 1>]}}}
      CurveIndex=$Particles.PartitionParticles.Normalized Index In Partition
     )
    )
    Write to Module Namespace=true
   )
   /SetVariables(Particles.ImpactTrailWidth:NiagaraFloat=&FloatFromCurve(
     FloatCurve=DataInterface() {propertyValues {Curve {keys [<at 0.5 1>,<at 1 1>]}}}
     CurveIndex=$Output.GetBulletParams.BulletProgressAlpha
     Scale Curve=15
    )) {isSetParameters=true}
  ]
  rendererClasses [&NiagaraRibbonRendererProperties]
  renderers [
   {rendererClass=&NiagaraRibbonRendererProperties}
   {rendererClass=&NiagaraRibbonRendererProperties, rendererIndex=1}
   {rendererClass=&NiagaraRibbonRendererProperties, rendererIndex=2}
  ]
 }
 NE_Bullets {
  emitterUpdateScript [
   &EmitterState(
    Inactive Response=::NewEnumerator1
    Life Cycle Mode=::NewEnumerator1
    Loop Behavior=::NewEnumerator1
    Loop Duration Mode=::NewEnumerator1
    UseLoopDelay=false
   )
   &SpawnBurst_Instantaneous(Spawn Count=$System.BulletNumber, Use Spawn Probability=false, UseLoopCountLimit=false)
  ]
  particleSpawnScript [
   &InitializeParticle(
    Color Mode=::NewEnumerator0
    Lifetime=5
    Mesh Scale Mode=::NewEnumerator3
    Mesh Uniform Scale=2
    Sprite Size=<25 45>
    Sprite Size Mode=::NewEnumerator4
    UsePositionOffset=false
    Write Material Random=true
   )
   /SetVariables() {isSetParameters=true}
   /SetVariables(Particles.ProgressAlpha:NiagaraFloat=&Add_Float(
     A=&Multiply_Float(A=&ReturnNormalizedExecIndex(Normalized Index Includes One=false), B=1)
     B=$System.StartDelay
    )) {isSetParameters=true}
   /SetVariables(Particles.FireDirection:Vector3f=&RandomVector(Vector Scale=1, Constrain to Hemisphere=true, Hemisphere Constraint Vector=<0 0 1>)) {isSetParameters=true}
   /SetVariables(Particles.Position=&Multiply_VectorByFloat(Vector=&Multiply_Vector(A=$Particles.FireDirection, B=<1 1 1>), Float=-750)) {isSetParameters=true}
   /SetVariables(Particles.GlowCameraOffset=20) {isSetParameters=true}
   /SetVariables(Particles.ImpactPosition:NiagaraPosition=$Particles.Position) {isSetParameters=true}
   /SetVariables(Particles.ImpactAlignment:Vector3f=$Particles.SpriteAlignment) {isSetParameters=true}
   &ShapeLocation(MaskSpawnGroup=false, Override Seed=false, UseOwnerScale=true)
  ]
  particleUpdateScript [
   &ParticleState(Kill Particles When Lifetime Has Elapsed=false)
   /SetVariables(Particles.SystemVelocity:Vector3f=&Lerp_Vector(
     A=$Particles.SystemVelocity
     B=&TransformVector(Vector=$Engine.Owner.Velocity, Source Space=::World, Destination Space=::Local)
     Alpha=&Multiply_Float(A=$Engine.DeltaTime, B=2)
    )) {isSetParameters=true}
   /SetVariables(Particles.FireDirection:Vector3f=&MakeCustomPositionFromBool(
     Bool:NiagaraBool=&SetBoolByFloatComparison(A=$Particles.ProgressAlpha, B=0.01, Comparison Type=::NewEnumerator6)
     True Position:NiagaraPosition=&RandomVector(Vector Scale=1, Constrain to Hemisphere=true, Hemisphere Constraint Vector=<0 0 1>, Evaluation Type=::NewEnumerator1)
     False Position:NiagaraPosition=Linked() {linkedVariable {name=Particles.FireDirection, type=Vector3f}}
    )) {isSetParameters=true}
   /SetVariables(Particles.BulletSpeed:NiagaraFloat=&Multiply_Float(
     A=&RandomRangeFloat(Minimum=1, Maximum=1)
     B=&Lerp_Float(A=0.8, B=1.2, Alpha=$System.SmoothDiveAlpha)
    )) {isSetParameters=true}
   /SetVariables(Particles.PreivousPos:NiagaraPosition=$Particles.Position) {isSetParameters=true}
   /SetVariables(Particles.ProgressAlpha:NiagaraFloat=&ModuloFloat(
     A=&Add_Float(A=$Particles.ProgressAlpha, B=&Multiply_Float(A=$Engine.DeltaTime, B=$System.BulletSpeed))
     B=1
    )) {isSetParameters=true}
   /SetVariables(Particles.FirstPhaseAlpha:NiagaraFloat=&FloatFromCurve(FloatCurve=DataInterface() {propertyValues {Curve {keys [<at 0.1 0>,<at 0.4 1>]}}}, CurveIndex=$Particles.ProgressAlpha)) {isSetParameters=true}
   /SetVariables(Particles.SecondPhaseAlpha:NiagaraFloat=&FloatFromCurve(FloatCurve=DataInterface() {propertyValues {Curve {keys [<at 0.4 0>,<at 1 1>]}}}, CurveIndex=$Particles.ProgressAlpha)) {isSetParameters=true}
   &RotateAroundPoint(
    Radius:NiagaraFloat=$System.SphereRadius
    Rotation Center:NiagaraPosition=$System.SphereOffset
    Rotation Phase:NiagaraFloat=&Lerp_Float(
     A=&Add_Float(A=-0.25, B=&RandomRangeFloat(Minimum=0, Maximum=0))
     B=&RandomRangeFloat(Minimum=-0.15, Maximum=-0.2)
     Alpha=$Particles.SecondPhaseAlpha
    )
    Set Particle Position=false
    X Axis:Vector3f=&NormalizeVector(
     Vector To Normalize=/Niagara/DynamicInputs/Vectors/CrossProduct(
      Vector A:Vector3f=&CrossProduct(
       Vector A:Vector3f=$Particles.FireDirection
       Vector B:Vector3f=&RandomRangeVector(Minimum=<0 0 0>, Maximum=<0 0 1>)
       Normalize Incoming Vectors:NiagaraBool=true
      )
      Vector B:Vector3f=$Particles.FireDirection
      Normalize Incoming Vectors:NiagaraBool=true
     )
     Safe Normalize=true
     Safe Fallback Threshold=0
     Safe Fallback Vector=<0 0 0.001>
    )
    Y Axis:Vector3f=$Particles.FireDirection
   )
   /SetVariables(Particles.Position=&Lerp_Vector(
     A=&Lerp_Vector(
      A=&Multiply_VectorByFloat(Vector=$Particles.FireDirection, Float=-750)
      B=$Output.RotateAroundPoint.OutputPosition
      Alpha=$Particles.FirstPhaseAlpha
     )
     B=&Add_Vector(
      A=&Add_Vector(
       A=$Output.RotateAroundPoint.OutputPosition
       B=&Multiply_VectorByFloat(
        Vector=&Multiply_Vector(A=$Particles.SystemVelocity, B=<-0.1 -0.1 0>)
        Float=&FloatFromCurve(FloatCurve=DataInterface() {propertyValues {Curve {keys [<at 0.5 0>,<at 1 1>]}}}, CurveIndex=$Particles.SecondPhaseAlpha)
       )
      )
      B=&Add_Vector(
       A=&Multiply_VectorByFloat(
        Vector=&Lerp_Vector(A=<0 0 150>, B=<0 0 100>, Alpha=$User.ParaGlideLeanAlpha)
        Float=&FloatFromCurve(
         FloatCurve=DataInterface() {propertyValues {Curve {keys [
             {arriveTangent=0.5818, interpMode=RCIM_Cubic, leaveTangent=0.5818, tangentMode=RCTM_User, time=0, value=0}
             {arriveTangent=1.7259, interpMode=RCIM_Cubic, leaveTangent=1.7259, tangentMode=RCTM_User, time=1, value=1}
            ]}}}
         CurveIndex=$Particles.SecondPhaseAlpha
        )
       )
       B=&Multiply_VectorByFloat(
        Vector=&NormalizeVector(
         Vector To Normalize=$Particles.RotateAroundPoint.VectorFromCenter
         Safe Normalize=true
         Safe Fallback Threshold=0
         Safe Fallback Vector=<0 0 0.001>
        )
        Float=&Multiply_Float(
         A=&RandomRangeFloat(Minimum=150, Maximum=250)
         B=&FloatFromCurve(
          FloatCurve=DataInterface() {propertyValues {Curve {keys [
              {arriveTangent=13.2873, interpMode=RCIM_Cubic, leaveTangent=13.2873, tangentMode=RCTM_User, time=0, value=0}
              {interpMode=RCIM_Cubic, tangentMode=RCTM_User, time=0.25, value=1}
             ]}}}
          CurveIndex=$Particles.SecondPhaseAlpha
         )
        )
       )
      )
     )
     Alpha=$Particles.SecondPhaseAlpha
    )) {isSetParameters=true}
   /SetVariables(Particles.MeshOrientation=&LerpQuaternion(
     Quaternion A:Quat4f=&OrientMeshToVector(
      Look At Direction=&NormalizeVector(
       Vector To Normalize=&Subtract_Position(A=$Particles.Position, B=$Particles.PreivousPos)
       Safe Normalize=true
       Safe Fallback Threshold=0
       Safe Fallback Vector=<0 0 0.001>
      )
      Reference Axis Vector=<1 0 0>
     )
     Quaternion B:Quat4f=&MultiplyQuaternion(
      Quaternion A=$Particles.MeshOrientation
      Quaternion B=&MakeQuatFromAxisAngle(
       Axis=&RandomVector(Vector Scale=1)
       AngleInDegrees=&RandomRangeFloat(Minimum=180, Maximum=360)
      )
     )
     Lerp Factor:NiagaraFloat=&FloatFromCurve(FloatCurve=DataInterface() {propertyValues {Curve {keys [<at 0 0>,<at 1 1>]}}}, CurveIndex=$Particles.SecondPhaseAlpha)
     Enable Spherical Linear Interpolation:NiagaraBool=false
    )) {enabled=false, isSetParameters=true}
   /SetVariables(Particles.MeshOrientation=&MultiplyQuaternion(
     Quaternion A=&OrientMeshToVector(
      Look At Direction=&NormalizeVector(
       Vector To Normalize=&Subtract_Position(A=$Particles.Position, B=$Particles.PreivousPos)
       Safe Normalize=true
       Safe Fallback Threshold=0
       Safe Fallback Vector=<0 0 0.001>
      )
      Reference Axis Vector=<1 0 0>
     )
     Quaternion B=&MakeQuatFromAxisAngle(
      Axis=&RandomVector(Vector Scale=1)
      AngleInDegrees=&Multiply_Float(
       A=&RandomRangeFloat(Minimum=280, Maximum=360)
       B=&FloatFromCurve(FloatCurve=DataInterface() {propertyValues {Curve {keys [<at 0 0>,<at 1 1>]}}}, CurveIndex=$Particles.SecondPhaseAlpha)
      )
     )
    )) {isSetParameters=true}
   /SetVariables(Particles.SpriteAlignment=&NormalizeVector(
     Vector To Normalize=&RotateVectorByQuaternion(Vector To Rotate=<1 0 0>, Quaternion=$Particles.MeshOrientation)
     Safe Normalize=true
     Safe Fallback Threshold=0
     Safe Fallback Vector=<0 0 0.001>
    )) {isSetParameters=true}
   &self:SetImpactPos(
    NewAlignment:Vector3f=&NormalizeVector(
     Vector To Normalize=$Particles.RotateAroundPoint.VectorFromCenter
     Safe Normalize=true
     Safe Fallback Threshold=0
     Safe Fallback Vector=<0 0 0.001>
    )
    ProgressAlpha:NiagaraFloat=$Particles.SecondPhaseAlpha
    RefreshThreshold=0.05
   ) {module=SetImpactPos}
   &Color(Color=&ColorFromCurve(
     ColorCurve=DataInterface() {propertyValues {
       AlphaCurve {keys [<at 0 1 RCIM_Cubic>]}
       BlueCurve {keys [<at 0 0 RCIM_Cubic>,<at 1 0>]}
       GreenCurve {keys [<at 0 0 RCIM_Cubic>,<at 1 0.4884>]}
       RedCurve {keys [<at 0 0 RCIM_Cubic>,<at 1 1.5>]}
      }}
     CurveIndex=&FloatFromCurve(FloatCurve=DataInterface() {propertyValues {Curve {keys [<at 0.4 0>,<at 0.42 1>,<at 0.8 0>]}}}, CurveIndex=$Particles.ProgressAlpha)
    ))
   &ScaleMeshSize(Scale Factor=&VectorFromFloat(Value=&Multiply_Float(
      A=&FloatFromCurve(
       FloatCurve=DataInterface() {propertyValues {Curve {keys [{tangentMode=RCTM_User, time=0.05, value=0}, {tangentMode=RCTM_User, time=0.1, value=1}, <at 0.8 1>, <at 1 0>]}}}
       CurveIndex=$Particles.ProgressAlpha
      )
      B=&ClampFloat(Float=&Divide_Float(
        A=&Subtract_Float(
         A=&DistanceBetweenPositions(
          A=&ReturnCameraPosition(Coordinate Space:ENiagaraCoordinateSpace=::Simulation, Camera Query:NiagaraDataInterfaceCamera=DataInterface())
          B=$Particles.Position
         )
         B=100
        )
        B=100
       ))
     )))
   &ScaleSpriteSize(
    Uniform Curve Index=$Particles.ProgressAlpha
    Uniform Curve Sprite Scale=DataInterface() {propertyValues {Curve {keys [{tangentMode=RCTM_User, time=0, value=0}, {tangentMode=RCTM_User, time=0.1, value=1}, <at 0.8 1>, <at 1 0>]}}}
   )
   &SolveForcesAndVelocity(Clamp Velocity=false, Limit Acceleration=false)
   /SetVariables(Particles.ImpactAlignment:Vector3f=&NormalizeVector(
     Vector To Normalize=$Particles.RotateAroundPoint.VectorFromCenter
     Safe Normalize=true
     Safe Fallback Threshold=0
     Safe Fallback Vector=<0 0 0.001>
    )) {enabled=false, isSetParameters=true}
   /SetVariables(Particles.ImpactSize:Vector2f=&Vector2DFromFloat(Value=&Lerp_Float(
      A=0
      B=100
      Alpha=&FloatFromCurve(FloatCurve=DataInterface() {propertyValues {Curve {keys [<at 0.4 0>,<at 0.45 1>,<at 0.6 0>]}}}, CurveIndex=$Particles.ProgressAlpha)
     ))) {isSetParameters=true}
   /SetVariables(Particles.ImpactColor:LinearColor=&ColorFromCurve(
     ColorCurve=DataInterface() {propertyValues {
       AlphaCurve {keys [<at 0 1 RCIM_Cubic>]}
       BlueCurve {keys [<at 0 0 RCIM_Cubic>,<at 1 0.0324>]}
       GreenCurve {keys [<at 0 0.1565 RCIM_Cubic>,<at 1 0.7261>]}
       RedCurve {keys [<at 0 0.2 RCIM_Cubic>,<at 1 2>]}
      }}
     CurveIndex=&FloatFromCurve(FloatCurve=DataInterface() {propertyValues {Curve {keys [<at 0.35 0>,<at 0.4 1>,<at 0.7 0>]}}}, CurveIndex=$Particles.ProgressAlpha)
    )) {isSetParameters=true}
   /SetVariables(Particles.MeshIndex=&MakeCustomIntFromBool(Bool=&SetBoolByFloatComparison(A=$Particles.FirstPhaseAlpha, B=0.99), True INT=0, False INT=0)) {isSetParameters=true}
   &DynamicMaterialParameters(
    Index 3 Param 1=&FloatFromCurve(
     FloatCurve=DataInterface() {propertyValues {Curve {keys [<at 0.48 1>,<at 0.5 1>,<at 1 0>]}}}
     CurveIndex=$Particles.ProgressAlpha
     Scale Curve=2
    )
    Module.Index3Float0Write=true
    Module.Index3Float1Write=true
    Module.Index3Float2Write=true
    Module.Index3Float3Write=true
    Write Parameter Index 0=::NewEnumerator0
    Write Parameter Index 3=::NewEnumerator1
   )
   &FN_Sprite_Attributes(
    Color=&MakeLinearColorFromVectorAndFloat(
     "Vector (RGB)"=&MakeVectorFromLinearColorRGB(Color=<rgba 1 0.8398 0.3594 1>)
     "Float (Alpha)"=&FloatFromCurve(
      FloatCurve=DataInterface() {propertyValues {Curve {keys [<at 0.4 0>,<at 0.45 1>,<at 0.8 0>]}}}
      CurveIndex=$Particles.ProgressAlpha
      Scale Curve=0.3
     )
    )
    Create Color Attribute=true
    Create SpriteSize Attribute=true
    SpriteSize=&Vector2DFromFloat(Value=&Multiply_Float(
      A=&RandomRangeFloat(Minimum=60, Maximum=80)
      B=&FloatFromCurve(FloatCurve=DataInterface() {propertyValues {Curve {keys [<at 0.4 0>,<at 0.8 1>]}}}, CurveIndex=$Particles.ProgressAlpha)
     ))
    Write to Module Namespace=true
   )
   /SetVariables(
    Particles.SmokeDynMat1:Vector4f=&MakeVector4(
     X=&FloatFromCurve(FloatCurve=DataInterface() {propertyValues {Curve {keys [<at 0.4 0>,<at 1 1>]}}}, CurveIndex=$Particles.ProgressAlpha)
     Y=0
     Z=0
     W=0
    )
    Particles.SmokeDynMat2:Vector4f=&MakeVector4(
     X=&FloatFromCurve(FloatCurve=DataInterface() {propertyValues {Curve {keys [<at 0.4 0>,<at 0.7 1>]}}}, CurveIndex=$Particles.ProgressAlpha)
     Y=0
     Z=1
     W=1.2
    )
   ) {isSetParameters=true}
   &FN_Mesh_Attributes(
    Color=&Multiply_LinearColorByFloat(
     Linear Color=<rgba 10 4.735 0.2107 2>
     Float=&FloatFromCurve(FloatCurve=DataInterface() {propertyValues {Curve {keys [<at 0 0>,<at 0.1 1>,<at 0.95 1>,<at 1 0>]}}}, CurveIndex=$Particles.FirstPhaseAlpha)
    )
    Create Color Attribute=true
    Create Scale Attribute=true
    Scale=<0.5 0.1 0.1>
    Write to Module Namespace=true
   )
  ]
  rendererClasses [&NiagaraSpriteRendererProperties, &NiagaraMeshRendererProperties]
  renderers [
   {rendererClass=&NiagaraSpriteRendererProperties}
   {rendererClass=&NiagaraMeshRendererProperties, rendererIndex=1}
   {rendererClass=&NiagaraSpriteRendererProperties, rendererIndex=2}
   {rendererClass=&NiagaraSpriteRendererProperties, rendererIndex=3}
   {rendererClass=&NiagaraMeshRendererProperties, rendererIndex=4}
  ]
 }
 Speedlines {
  emitterUpdateScript [
   &EmitterState()
   &SpawnRate(SpawnRate=20, Use Spawn Probability=true) {module=SpawnRate001}
  ]
  particleSpawnScript [
   &InitializeParticle(
    Color=&ColorFromCurve(
     ColorCurve=DataInterface() {propertyValues {
       AlphaCurve {keys [<at 0 1 RCIM_Cubic>]}
       BlueCurve {keys [<at 0.4214 0.4062 RCIM_Cubic>]}
       GreenCurve {keys [<at 0.4214 1.4765 RCIM_Cubic>]}
       RedCurve {keys [<at 0.4214 2 RCIM_Cubic>]}
      }}
     CurveIndex=&RandomRangeFloat(Minimum=0, Maximum=1)
    )
    Lifetime Max=0.4
    Lifetime Min=0.3
    Lifetime Mode=::NewEnumerator1
    Position Offset=&FN_FrontEndVector3Modifier(Gameplay=<0 0 0>, Front End=<0 50 0>)
    Sprite Size Max=<3 350>
    Sprite Size Min=<2 250>
    Sprite Size Mode=::NewEnumerator2
    UsePositionOffset=true
    Write Material Random=true
   )
   &ShapeLocation(
    Additional Quaternion Rotation=true
    MaskSpawnGroup=false
    Offset=&Add_Vector(A=&MakeVector(X=0, Y=0, Z=-300), B=$System.SphereOffset)
    Offset Coordinate Space=::Local
    Offset Mode=::NewEnumerator0
    Rotation Mode=::NewEnumerator3
    Shape Primitive=::NewEnumerator3
    Torus Distribution Mode:ENiagaraTorusDistributionMode=::NewEnumerator1
    U Position=&RandomRangeFloat(Minimum=0, Maximum=1)
    UseOwnerScale=true
    V Position=1
   )
   &AddVelocity(Velocity=&Multiply_VectorByFloat(
     Vector=&NormalizeVector(Vector To Normalize=<0 0 -1>, Safe Normalize=true, Safe Fallback Threshold=0, Safe Fallback Vector=<0 0 0.001>)
     Float=&RandomRangeFloat(Minimum=-1200, Maximum=-1500)
    ))
   &AddVelocity(
    Velocity=&NormalizeVector(Vector To Normalize=$Particles.ShapeLocation.ShapeNormal, Safe Normalize=true, Safe Fallback Threshold=0, Safe Fallback Vector=<0 0 0.001>)
    Velocity Speed Scale=&RandomRangeFloat(Minimum=250, Maximum=300)
   ) {module=AddVelocity001}
  ]
  particleUpdateScript [
   &ParticleState()
   /SetVariables(Particles.PreviousPos:NiagaraPosition=$Particles.Position) {isSetParameters=true}
   &ScaleSpriteSize(Scale Factor=&Vector2DFromCurve(
     Vector2Curve=DataInterface() {propertyValues {XCurve {keys [<at 0 0.5 RCIM_Cubic>,<at 0.5 1 RCIM_Cubic>]}, YCurve {keys [<at 0 0 RCIM_Cubic>,<at 0.5 1 RCIM_Cubic>]}}}
     ScaleCurve=&Vector2DFromFloat(Value=1)
    ))
   &AccelerationForce(Acceleration=&Multiply_VectorByFloat(
     Vector=&Multiply_Vector(A=$Engine.Owner.Velocity, B=<0.5 0.5 0>)
     Float=&FN_FrontEndFloatModifier(Gameplay=1, Front End=0)
    ))
   &SolveForcesAndVelocity(Clamp Velocity=false, Limit Acceleration=false)
   &ScaleColor(
    Scale Alpha=&Multiply_Float(
     A=&FloatFromCurve(FloatCurve=DataInterface() {propertyValues {Curve {keys [<at 0 0 RCIM_Cubic>, <at 0.2 1>, {arriveTangent=-2.221, interpMode=RCIM_Cubic, leaveTangent=-2.221, tangentMode=RCTM_User, time=1, value=0}]}}})
     B=&FloatFromCurve(
      FloatCurve=DataInterface() {propertyValues {Curve {keys [
          <at 0 0 RCIM_Cubic>
          <at 0.5 0>
          {arriveTangent=5.9607, interpMode=RCIM_Cubic, leaveTangent=5.9607, tangentMode=RCTM_User, time=0.8, value=1}
         ]}}}
      CurveIndex=$User.ParaGlideLeanAlpha
      Scale Curve=2
     )
    )
    ScaleA=true
    ScaleRGB=true
   )
  ]
  rendererClasses [&NiagaraSpriteRendererProperties]
  renderers [{rendererClass=&NiagaraSpriteRendererProperties}]
 }
}
summary {
 systemName=NS_Contrail_AuraCop
 emitters [
  {emitterName=NE_BulletRibbons, rendererClasses [&NiagaraRibbonRendererProperties]}
  {emitterName=Speedlines, rendererClasses [&NiagaraSpriteRendererProperties]}
  {emitterName=NE_Bullets, rendererClasses [&NiagaraSpriteRendererProperties, &NiagaraMeshRendererProperties]}
  {emitterName=Energy_Sphere, rendererClasses [&NiagaraMeshRendererProperties, &NiagaraSpriteRendererProperties]}
 ]
 userVariables [
  User.bIsFrontEnd {type=NiagaraBool}
  User.bIsFrontEndPreview {type=NiagaraBool}
  User.ParaGlideLeanAlpha {type=NiagaraFloat, defaultValue=0, description=""}
 ]
}
